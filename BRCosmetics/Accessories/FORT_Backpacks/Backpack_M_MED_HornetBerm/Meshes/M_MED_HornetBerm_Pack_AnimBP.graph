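# Animation Blueprint for the HornetBerm backpack cosmetic that repositions its pack bone as the wearer's spine bends and twists.
# Each update maps the owner mesh's spine_05-to-spine_01 rotation delta from ref pose through clamped ranges into translation, rotation and scale variables.
# The AnimGraph stacks additive Modify Bone nodes on pack, layering fixed offsets and those variable-driven adjustments.
# The update also computes an eased skydive-dive value and a vertical-velocity "trampoline" value.
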AnimBlueprint /BRCosmetics/Accessories/FORT_Backpacks/Backpack_M_MED_HornetBerm/Meshes/M_MED_HornetBerm_Pack_AnimBP
target_skeleton=&/M_MED_HornetBerm_Pack_Skeleton
graphs [
 AnimGraph {nodes [
   n1 = Input Pose()
   n2 = "Transform (Modify) Bone - Bone: pack"(
    ComponentPose="Transform (Modify) Bone - Bone: pack"(ComponentPose=Local To Component(LocalPose=@n1), Translation="0,2.5,0", Rotation="0,0,0") {translationMode=BMM_Additive, boneToModify {boneName=pack}, translation {y=2.5}}
    Translation="0,0,0"
    Rotation="0,-10,5"
   ) {rotationMode=BMM_Additive, translationMode=BMM_Additive, boneToModify {boneName=pack}, rotation {roll=5, yaw=-10}}
   n3 = "Transform (Modify) Bone - Bone: pack"(ComponentPose=@n2, Translation="0,-2,0", Rotation="5,0,5") {rotationMode=BMM_Additive, translationMode=BMM_Additive, boneToModify {boneName=pack}, rotation {pitch=5, roll=5}, translation {y=-2}}
   n4 = "Transform (Modify) Bone - Bone: pack"!(ComponentPose=@n3, bAlphaBoolEnabled:bool, Translation="0,-2,0", Rotation="0,0,5", Scale) -> (Pose) {
    alphaInputType=Bool
    rotationMode=BMM_Additive
    translationMode=BMM_Additive
    boneToModify {boneName=pack}
    rotation {roll=5}
    translation {y=-2}
   }
   n5 = "Transform (Modify) Bone - Bone: pack"!(
    ComponentPose="Transform (Modify) Bone - Bone: pack"(
     ComponentPose=@n4
     Translation=Get backpack_translation_r() -> (backpack_translation_r:Vector)
     Rotation=Get backpack_rotate()
     Scale=Get backpack_scale() -> (backpack_scale:Vector)
    ) {rotationMode=BMM_Additive, scaleMode=BMM_Additive, translationMode=BMM_Additive, boneToModify {boneName=pack}}
    Alpha:float
    Translation=Get backpack_translation_l() -> (backpack_translation_l:Vector)
    Rotation=Get pickaxe_rotate_l() -> (pickaxe_rotate_l:Rotator)
    Scale
   ) -> (Pose) {
    rotationMode=BMM_Additive
    rotationSpace=BCS_BoneSpace
    translationMode=BMM_Additive
    translationSpace=BCS_BoneSpace
    boneToModify {boneName=pack}
   }
   n6 = Get pickaxe_translate() -> (pickaxe_translate:Vector)
   n7 = Get pickaxe_rotate_r() -> (pickaxe_rotate_r:Rotator)
   Output Pose(Result=Component To Local(ComponentPose="Transform (Modify) Bone - Bone: pack"(ComponentPose=@n5, Translation=@n6, Rotation=@n7) {rotationMode=BMM_Additive, translationMode=BMM_Additive, boneToModify {boneName=pack}})) {layerGroup=None}
  ]}
 EventGraph {nodes [
   n1 = Event BlueprintUpdateAnimation()
   > n2 = Sequence!(execute) -> (then_0, then_1, then_2:Exec, then_3:Exec, then_4:Exec, then_5:Exec, then_6:Exec, then_7:Exec, then_8:Exec)
   n3 = Get bIsSkydiveDiveMode()
   Set Skydive Dive Mode(
    execute=@n2.then_8
    Skydive Dive Mode:double=FInterpTo(
     Current=Get Skydive Dive Mode() -> (Skydive Dive Mode:double)
     Target=SelectFloat(A=1, B=0, bPickA=@n3)
     DeltaTime=@n1.DeltaTimeX
     InterpSpeed=5
    )
   ) -> (Output_Get:double)
   n4 = GetOwnerSkeletalMeshComponent()
   n5 = BreakRotator(InRot=BreakTransform(InTransform=GetDeltaTransformFromRefPose(self=@n4, BoneName=spine_05, BaseName=spine_01)).Rotation)
   Set!(
    execute=@n2.then_6
    pickaxe_rotate_l_Roll:float=MapRangeClamped(Value=@n5, InRangeA=-35, InRangeB=-45, OutRangeB=10)
    pickaxe_rotate_l_Pitch:float=0
    pickaxe_rotate_l_Yaw:float=0
   ) -> (then, Output_Get:Rotator)
   n6 = MapRangeClamped(Value=@n5.Yaw, InRangeA=0, InRangeB=-60, OutRangeA=1, OutRangeB=0.85)
   Set backpack_scale(execute=@n2.then_2, backpack_scale:Vector=MakeVector(X=@n6, Y=@n6, Z=@n6)) -> (Output_Get:Vector)
   Set!(
    execute=@n2.then_0
    backpack_translation_r_X:double=MapRangeClamped(Value=@n5.Yaw, InRangeA=-15, InRangeB=-65, OutRangeB=0)
    backpack_translation_r_Y:double=MapRangeClamped(Value=@n5.Yaw, InRangeA=-15, InRangeB=-65, OutRangeB=-15)
    backpack_translation_r_Z:double=MapRangeClamped(Value=@n5.Yaw, InRangeA=-15, InRangeB=-65, OutRangeB=3)
   ) -> (then, Output_Get:Vector)
   Set trampoline(execute=@n2.then_5, trampoline:double=MapRangeClamped(Value=Get PawnVelocityZ(), InRangeA=0, InRangeB=4000)) -> (Output_Get:double)
   Set!(
    execute=@n2.then_7
    backpack_translation_l_X:double=0
    backpack_translation_l_Y:double=MapRangeClamped(Value=@n5, InRangeA=-15, InRangeB=-45, OutRangeB=-3)
    backpack_translation_l_Z:double=0
   ) -> (then, Output_Get)
   Set!(
    execute=@n2.then_3
    pickaxe_rotate_r_Roll:float=MapRangeClamped(Value=@n5, InRangeA=20, InRangeB=50, OutRangeB=0)
    pickaxe_rotate_r_Pitch:float=0
    pickaxe_rotate_r_Yaw:float=MapRangeClamped(Value=@n5, InRangeA=20, InRangeB=50, OutRangeB=-20)
   ) -> (then, Output_Get:Rotator)
   Set!(
    execute=@n2.then_4
    pickaxe_translate_X:double=0
    pickaxe_translate_Y:double=MapRangeClamped(Value=@n5, InRangeA=20, InRangeB=50, OutRangeB=-3)
    pickaxe_translate_Z:double=0
   ) -> (then, Output_Get:Vector)
   Set!(
    execute=@n2.then_1
    backpack_rotate_Roll:float=MapRangeClamped(Value=@n5.Yaw, InRangeA=-20, InRangeB=-70, OutRangeB=70)
    backpack_rotate_Pitch:float=MapRangeClamped(Value=@n5.Yaw, InRangeA=-25, InRangeB=-70, OutRangeB=-20)
    backpack_rotate_Yaw:float=MapRangeClamped(Value=@n5.Yaw, InRangeA=-25, InRangeB=-70, OutRangeB=-60)
   ) -> (then, Output_Get:Rotator)
   TryGetPawnOwner()
  ]}
]
variables [
 backpack_rotate {type=Rotator}
 backpack_scale {type=Vector}
 backpack_translation_l {type=Vector}
 backpack_translation_r {type=Vector}
 is On Ground {type=boolean}
 pickaxe_rotate_l {type=Rotator}
 pickaxe_rotate_r {type=Rotator}
 pickaxe_translate {type=Vector}
 Skydive Dive Mode {type=number}
 trampoline {type=number}
]
